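# Lit material for a jetpack fuel gauge, showing the FuelLevel parameter as a segmented emissive bar.
# A gradient blends two or three colors from VTech to out-of-fuel, plus a bump-offset center glow for parallax depth.
# Repeating SphereMask bars and a rounded, fuel-shifted gradient cutoff mask that color before emissive adjustment.
# Both IsHorizontal? switch branches read UGradient, and a camera offset applies only on the mobile shading path.
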Material /Game/Accessories/FORT_Backpacks/Materials/M_Jetpack_FuelLevel
graph {
 properties {bUsedWithMeshDeformer=true, bUsedWithSkeletalMesh=true, pixelDepthOffsetMode=PDOM_Legacy, usedShadingModels=Default Lit}
 nodes [
  n1 = VectorParameter() {group=Color Control, parameterName=VTechColor, sortPriority=2, defaultValue=<rgba 0 1.5 1.5 1>}
  n2 = VectorParameter() {group=Color Control, parameterName=OutOfFuel, sortPriority=4, defaultValue=<rgba 5 0 0 1>}
  n3 = MaterialFunctionCall(UV Channel=StaticSwitchParameter(True=TextureCoordinate() {coordinateIndex=1}, False=TextureCoordinate()) {defaultValue=true, parameterName=UseFirstTexCoord}) {materialFunction=&LinearGradient}
  n4 = MaterialFunctionCall(
   In=StaticSwitchParameter(True=@n3.UGradient, False=@n3.UGradient) {parameterName=IsHorizontal?}
   Contrast=ScalarParameter() {defaultValue=1, parameterName=GradientContrast}
  ) {materialFunction=&CheapContrast}
  n5 = LinearInterpolate(
   A=LinearInterpolate(A=@n1.RGB, B=@n2.RGB, Alpha=@n4.Result)
   B=MaterialFunctionCall(
    1=@n1.RGB
    2=VectorParameter() {group=Color Control, parameterName=VTechColor-Mid, sortPriority=3, defaultValue=<rgba 0.5 1 0 1>}.RGB
    3=@n2.RGB
    A=@n4.Result
   ) {materialFunction=&Lerp_Multiple_Float3}.Lerp 3 Inputs
   Alpha=ScalarParameter() {group=Color Control, parameterName=Use Center Color, sortPriority=1}
  )
  n6 = ScalarParameter() {defaultValue=0.44, group=Color Control, parameterName=CenterGlow_Bias}
  n7 = MaterialFunctionCall(Float2=StaticSwitchParameter(True=TextureCoordinate() {coordinateIndex=1}, False=TextureCoordinate()) {defaultValue=true, parameterName=UseFirstTexCoord}) {materialFunction=&BreakOutFloat2Components}
  n8 = BumpOffset(Coordinate=@n7.R, Height=ScalarParameter() {defaultValue=-0.75}) {referencePlane=0}
  n9 = Multiply(
   A=SphereMask(
    A=Add(A=@n6, B=@n8)
    B=Constant() {r=0.5}
    Radius=ScalarParameter() {defaultValue=0.065, group=Color Control, parameterName=CenterGlow_Radius}
    Hardness=ScalarParameter() {group=Color Control, parameterName=CenterGlow_Hardness}
   )
   B=ScalarParameter() {defaultValue=0.075, group=Color Control, parameterName=Center Glow Influence, sortPriority=31}
  )
  LinearInterpolate(A=@n5, B=VectorParameter() {parameterName=Center Color, defaultValue=<rgba 1 1 1 1>}.RGB, Alpha=@n9)
  n10 = Add(A=@n5, B=@n9)
  n11 = Multiply(A=@n7.G, B=ScalarParameter() {defaultValue=1, parameterName=NumberOfBars})
  n12 = Power(
   Base=SphereMask(
    A=Frac(None=@n11)
    B=Constant() {r=0.5}
    Radius=ScalarParameter() {defaultValue=0.46, parameterName=Radius}
    Hardness=ScalarParameter() {defaultValue=0.65, parameterName=Hardness}
   )
   Exp=ScalarParameter() {defaultValue=1, parameterName=Bar Power}
  )
  n13 = MaterialFunctionCall(UV Channel=StaticSwitchParameter(True=TextureCoordinate() {coordinateIndex=1}, False=TextureCoordinate()) {defaultValue=true, parameterName=UseFirstTexCoord}) {materialFunction=&LinearGradient}
  n14 = ScalarParameter() {defaultValue=1.0093, parameterName=FuelLevel}
  n15 = Clamp(None=Add(
    A=StaticSwitchParameter(True=@n13.UGradient, False=@n13.UGradient) {parameterName=IsHorizontal?}
    B=LinearInterpolate(Alpha=@n14) {constA=-0.5, constB=0.5}
   ))
  MaterialFunctionCall(Emissive=Multiply(A=Multiply(A=@n10, B=@n12), B=Round(None=@n15))) {materialFunction=&MF_FX_EmissiveAdjustments}
  ScalarParameter() {parameterName=Metallic}
  ScalarParameter() {defaultValue=0.5, parameterName=Roughness}
  ScalarParameter() {defaultValue=0.5, parameterName=Specular}
  ShadingPathSwitch(
   Default=Constant()
   Mobile=MaterialFunctionCall(Offset Amount=ScalarParameter() {defaultValue=1, parameterName=CameraOffset}) {materialFunction=/Engine/Functions/Engine_MaterialFunctions02/WorldPositionOffset/CameraOffset}.Camera Offset
  ) {desc=Migrated from FeatureLevelSwitch}
 ]
}
