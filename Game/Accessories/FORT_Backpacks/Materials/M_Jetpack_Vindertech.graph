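# Material for the Vindertech jetpack backpack on skeletal meshes, drawing a glowing, segmented fuel indicator.
# A sphere mask over repeated second-UV V coordinates forms the segments, feeding an emissive adjustment function.
# FuelLevel against LowFuelLevel switches all segments between steady Vindertech blue and red that blinks on a rounded time sine.
# Refueling blink and stepped-sweep math, plus a refuel color, are present but do not feed that emissive chain.
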Material /Game/Accessories/FORT_Backpacks/Materials/M_Jetpack_Vindertech
graph {
 properties {bUsedWithMeshDeformer=true, bUsedWithSkeletalMesh=true, pixelDepthOffsetMode=PDOM_Legacy, usedShadingModels=Default Lit}
 nodes [
  n1 = ScalarParameter() {parameterName=Refueling}
  n2 = Multiply(A=Time(), B=ScalarParameter() {defaultValue=0.5, parameterName=RefuelSpeed})
  n3 = ConstantBiasScale(None=Sine(None=@n2) {period=0.25})
  Clamp(None=Add(A=OneMinus(None=@n1), B=Round(None=@n3)))
  n4 = ScalarParameter() {defaultValue=1, parameterName=FuelLevel}
  n5 = ScalarParameter() {defaultValue=0.2, parameterName=LowFuelLevel}
  n6 = Multiply(
   A=ScalarParameter() {defaultValue=1, parameterName=FullBlink}
   B=VectorParameter() {parameterName=VTechColor, defaultValue=<rgba 0.45 3.36 15 1>}.RGB
  )
  n7 = VectorParameter() {parameterName=OutOfFuel, defaultValue=<rgba 20 1 1 1>}
  n8 = Constant() {r=0.5}
  n9 = ScalarParameter() {defaultValue=0.45, parameterName=Radius}
  n10 = ScalarParameter() {defaultValue=0.5, parameterName=Hardness}
  n11 = TextureCoordinate() {coordinateIndex=1}
  n12 = ScalarParameter() {defaultValue=3, parameterName=Steps}
  n13 = Frac(None=Multiply(
    A=MaterialFunctionCall(Float2=@n11) {materialFunction=&BreakOutFloat2Components}.G
    B=@n12
   ))
  n14 = Time()
  n15 = Round(None=ConstantBiasScale(None=Sine(None=@n14) {period=0.25}))
  n16 = Multiply(
   A=Multiply(A=If(A=@n4, B=@n5, "A > B"=@n6, "A == B"=@n7.RGB, "A < B"=@n7.RGB), B=SphereMask(A=@n13, B=@n8, Radius=@n9, Hardness=@n10))
   B=If(A=@n4, B=@n5, "A > B"=Constant() {r=1}, "A == B"=@n15, "A < B"=@n15)
  )
  n17 = Constant() {r=1}
  n18 = If(A=@n4, B=@n5, "A > B"=Constant() {r=1}, "A == B"=@n17, "A < B"=@n17)
  MaterialFunctionCall(Emissive=Multiply(A=Multiply(A=@n16, B=@n18), B=Constant() {r=0.2})) {materialFunction=&MF_FX_EmissiveAdjustments}
  n19 = Multiply(
   A=MaterialFunctionCall(Float2=TextureCoordinate() {coordinateIndex=1}) {materialFunction=&BreakOutFloat2Components}.G
   B=@n12
  )
  n20 = Subtract(A=@n12)
  n21 = Multiply(A=Divide(A=Floor(None=@n19), B=@n12), B=Divide(B=Divide(A=@n20, B=@n12)))
  n22 = LinearInterpolate(Alpha=Multiply(A=Frac(None=@n2)) {constB=2}) {constA=1, constB=0}
  n23 = Frac(None=Clamp(None=Add(A=@n21, B=@n22)))
  OneMinus(None=Clamp(None=Multiply(A=@n23) {constB=2}))
  Round(None=@n23)
  VectorParameter() {parameterName=ReFuelColor, defaultValue=<rgba 2 0.9723 0 1>}
 ]
}
