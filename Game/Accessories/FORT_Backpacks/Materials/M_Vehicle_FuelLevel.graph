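# Lit fuel gauge material for skeletal meshes, drawing a horizontal or vertical bar driven by a FuelLevel parameter.
# Contrast-adjusted UV gradient blends VTechColor to OutOfFuel, shaped by two multiplied SphereMasks and cut off by rounding the fuel-offset UV.
# Below LowFuelThreshold, a sine-pulsed flash color is added, strongest at the edges.
# Enable scales both this emissive, routed through MF_FX_EmissiveAdjustments, and a dark edge-to-center base shade.
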Material /Game/Accessories/FORT_Backpacks/Materials/M_Vehicle_FuelLevel
graph {
 properties {bUsedWithMeshDeformer=true, bUsedWithSkeletalMesh=true, pixelDepthOffsetMode=PDOM_Legacy, usedShadingModels=Default Lit}
 nodes [
  n1 = ComponentMask(None=TextureCoordinate()) {g=true}
  n2 = Multiply(A=Time(), B=ScalarParameter() {defaultValue=-0.5, parameterName=PanSpeed})
  ConstantBiasScale(None=Sine(None=Add(A=@n1, B=@n2)) {period=0.1})
  Constant() {r=0.05}
  MaterialFunctionCall(Speed=Constant() {r=0.1}) {materialFunction=&DebugTimeSine}
  n3 = VectorParameter() {parameterName=VTechColor, defaultValue=<rgba 0.0618 0.4635 0 1>}
  n4 = VectorParameter() {parameterName=OutOfFuel, defaultValue=<rgba 0.5 0 0 1>}
  n5 = TextureCoordinate()
  n6 = ComponentMask(None=@n5) {r=true}
  n7 = ComponentMask(None=@n5) {g=true}
  n8 = MaterialFunctionCall(
   In=Power(Base=StaticSwitchParameter(True=@n6, False=@n7) {parameterName=IsHorizontalBar?})
   Contrast=ScalarParameter() {defaultValue=1, parameterName=GradientContrast}
  ) {materialFunction=&CheapContrast}
  n9 = TextureCoordinate()
  n10 = ComponentMask(None=@n9) {g=true}
  n11 = ComponentMask(None=@n9) {r=true}
  n12 = Constant() {r=0.5}
  n13 = Multiply(
   A=SphereMask(
    A=StaticSwitchParameter(True=@n10, False=@n11) {parameterName=IsHorizontalBar?}
    B=@n12
    Radius=ScalarParameter() {defaultValue=0.3, parameterName=XRadius}
    Hardness=ScalarParameter() {defaultValue=0.5, parameterName=XHardness}
   ) {attenuationRadius=0.35}
   B=SphereMask(
    A=StaticSwitchParameter(True=@n11, False=@n10) {parameterName=IsHorizontalBar?}
    B=@n12
    Radius=ScalarParameter() {defaultValue=0.45, parameterName=YRadius}
    Hardness=ScalarParameter() {defaultValue=0.75, parameterName=YHardness}
   ) {attenuationRadius=0.5}
  )
  n14 = TextureCoordinate()
  n15 = ScalarParameter() {defaultValue=1, parameterName=FuelLevel}
  n16 = Add(
   A=StaticSwitchParameter(True=ComponentMask(None=@n14) {r=true}, False=ComponentMask(None=@n14) {g=true}) {parameterName=IsHorizontalBar?}
   B=LinearInterpolate(A=ScalarParameter() {defaultValue=-0.45, parameterName=FuelMin}, B=ScalarParameter() {defaultValue=0.48, parameterName=FuelMax}, Alpha=@n15) {constA=-0.5, constB=0.5}
  )
  n17 = Multiply(A=Multiply(A=LinearInterpolate(A=@n3.RGB, B=@n4.RGB, Alpha=@n8.Result), B=@n13), B=Round(None=Clamp(None=@n16)))
  n18 = Constant()
  n19 = LinearInterpolate(Alpha=Power(Base=@n13) {constExponent=0.55}) {constA=1, constB=0.15}
  n20 = VectorParameter() {parameterName=LowFuelFlashColor, defaultValue=<rgba 0.5 0.0206 0.0156 0>}
  n21 = Sine(None=Time()) {period=0.32}
  n22 = If(
   A=@n15
   B=ScalarParameter() {defaultValue=0.2, parameterName=LowFuelThreshold}
   "A > B"=@n18
   "A == B"=@n18
   "A < B"=Multiply(A=Multiply(A=@n19, B=@n20.RGB), B=ConstantBiasScale(None=@n21))
  )
  n23 = ScalarParameter() {defaultValue=1, parameterName=Enable}
  MaterialFunctionCall(Emissive=Multiply(A=Add(A=@n17, B=@n22), B=@n23)) {materialFunction=&MF_FX_EmissiveAdjustments}
  Multiply(
   A=LinearInterpolate(
    A=ScalarParameter() {defaultValue=0.1, parameterName=EdgeBaseColor}
    B=ScalarParameter() {defaultValue=0.04, parameterName=CenterBaseColor}
    Alpha=Power(Base=@n13)
   )
   B=@n23
  )
  OneMinus(None=@n9)
  OneMinus(None=@n5)
  OneMinus(None=@n14)
 ]
}
