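# M_VertexTest is a Default Lit backpack accessory material, apparently a test, that offsets vertices along a fixed direction.
# It normalizes the Direction vector parameter and scales it by the Normal scalar parameter, which acts as a distance despite its name.
# The result sets only World Position Offset, then passes through the MF_TechArtOverrides_accessories function.
# The material is enabled for skeletal meshes and mesh deformers.
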Material /Game/Accessories/FORT_Backpacks/Materials/M_VertexTest
graph {
 properties {
  bUseMaterialAttributes=true
  bUsedWithMeshDeformer=true
  bUsedWithSkeletalMesh=true
  pixelDepthOffsetMode=PDOM_Legacy
  usedShadingModels=Default Lit
 }
 nodes [
  n1 = Normalize(VectorInput=VectorParameter() {parameterName=Direction, defaultValue=<rgba 0 1 0 1>}.RGB)
  n2 = ScalarParameter() {defaultValue=1, parameterName=Normal}
  MaterialFunctionCall(In=SetMaterialAttributes!(MaterialAttributes, World Position Offset=Multiply(A=@n1, B=@n2)) {attributeSetTypes [#1]}) {materialFunction=&MF_TechArtOverrides_accessories}
 ]
}
