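# Physics asset for the backpack skeletal mesh SK_Backpack_32, a Fortnite-style back accessory.
# It defines a single rotated box body on the SK_BackPack_32 bone, offset to enclose the pack.
# No constraints are present, so the backpack collides as one rigid shape with no jointed or secondary motion from this asset.
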PhysicsAsset /Game/Accessories/FORT_Backpacks/Mesh/SK_Backpack_32_Physics
bodies {SK_BackPack_32 [{shapeType=Box, extentX=29.7019, extentY=46.2519, extentZ=53.7614, center=<-4.5256 -15.4364 -12.1024>, rotation=<rot -10.7417 -90 4.0572>}]}
constraints []
