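# Physics asset for the male Commando 11 backpack skeletal mesh, with capsule bodies for the root, a spine bone and a three-bone tail, apparently for secondary motion.
# The tail bones form a constraint chain from the root, each joint locking twist and swing1 so it bends on the swing2 axis only.
# The spine attaches separately to the root with twist locked and both swing axes left unlocked.
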PhysicsAsset /Game/Accessories/FORT_Backpacks/Mesh/SK_Backpack_Male_Commando_11_Physics
bodies {
 dyn_spine [{shapeType=Capsule, length=17.7448, radius=5.1457, center=<6.1222 -2.702 0.0351>, rotation=<rot 0 -4.0225 -89.8983>}]
 dyn_tail_01 [{shapeType=Capsule, length=5.3733, radius=4.8156, center=<5.4895 -0.7523 0.0172>, rotation=<rot -0.4583 82.9244 -88.3938>}]
 dyn_tail_02 [{shapeType=Capsule, length=3.499, radius=3.0458, center=<5.5658 -0.3895 0.0475>, rotation=<rot 0 85.7642 -89.4473>}]
 dyn_tail_03 [{shapeType=Capsule, length=2.25, radius=2.5, center=<4.2764 0 0.0001>, rotation=<rot 0 -90 90>}]
 root [{shapeType=Capsule, length=19.2865, radius=7.7419, center=<-0.0005 -4.826 -0.1437>}]
}
constraints [
 {
  bone1Name=dyn_tail_01
  bone2Name=root
  swing1LimitDegrees=10
  swing1Motion=Locked
  swing2LimitDegrees=10
  twistLimitDegrees=2
  twistMotion=Locked
 }
 {
  bone1Name=dyn_tail_02
  bone2Name=dyn_tail_01
  swing1LimitDegrees=10
  swing1Motion=Locked
  swing2LimitDegrees=10
  twistLimitDegrees=2
  twistMotion=Locked
 }
 {
  bone1Name=dyn_tail_03
  bone2Name=dyn_tail_02
  swing1LimitDegrees=10
  swing1Motion=Locked
  swing2LimitDegrees=10
  twistLimitDegrees=2
  twistMotion=Locked
 }
 {bone1Name=dyn_spine, bone2Name=root, swing1LimitDegrees=1, swing2LimitDegrees=20, twistLimitDegrees=45, twistMotion=Locked}
]
