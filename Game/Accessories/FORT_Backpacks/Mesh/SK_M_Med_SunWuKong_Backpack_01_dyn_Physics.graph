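# Physics asset for the Sun Wukong backpack cosmetic, giving its four flag-and-pole pieces simulated secondary motion.
# Each body is a single capsule: a small one on Root, plus one for each inner and outer, left and right pole and flag bone.
# Constraints form four two-link chains, each pole to Root and each flag to its pole.
# All joints share the same tight swing and twist limits, keeping sway small.
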PhysicsAsset /Game/Accessories/FORT_Backpacks/Mesh/SK_M_Med_SunWuKong_Backpack_01_dyn_Physics
bodies {
 Root [{shapeType=Capsule, length=1, radius=1, center=<0 -1.5728 -6.5429>}]
 dyn_flag_inner_l [{shapeType=Capsule, length=5.25, radius=5, center=<3.8886 -0.2948 4.8626>, rotation=<rot -85.663 85.6796 -90.0156>}]
 dyn_flag_inner_r [{shapeType=Capsule, length=5.25, radius=5, center=<4.4506 -0.3376 -5.0964>, rotation=<rot -85.6629 85.6795 -90.0155>}]
 dyn_flag_outer_l [{shapeType=Capsule, length=5.25, radius=5, center=<4.1497 -0.3146 0.2355>, rotation=<rot -85.6629 85.6795 -90.0156>}]
 dyn_flag_outer_r [{shapeType=Capsule, length=5.25, radius=5, center=<4.1497 -0.3146 0.2355>, rotation=<rot -85.6629 85.6795 -90.0156>}]
 dyn_pole_inner_l [{shapeType=Capsule, length=4.75, radius=4.8878, center=<3.3466 -0.2534 0.19>, rotation=<rot -85.664 85.6606 -89.9963>}]
 dyn_pole_inner_r [{shapeType=Capsule, length=4.75, radius=4.8878, center=<3.3466 -0.2534 0.19>, rotation=<rot -85.664 85.6606 -89.9963>}]
 dyn_pole_outer_l [{shapeType=Capsule, length=4.75, radius=4.5307, center=<3.3466 -0.2534 0.19>, rotation=<rot -85.664 85.6606 -89.9963>}]
 dyn_pole_outer_r [{shapeType=Capsule, length=4.75, radius=4.5307, center=<3.3466 -0.2534 0.19>, rotation=<rot -85.664 85.6606 -89.9963>}]
}
constraints [
 {bone1Name=dyn_pole_outer_l, bone2Name=Root, swing1LimitDegrees=10, swing2LimitDegrees=10, twistLimitDegrees=10}
 {bone1Name=dyn_flag_outer_l, bone2Name=dyn_pole_outer_l, swing1LimitDegrees=10, swing2LimitDegrees=10, twistLimitDegrees=10}
 {bone1Name=dyn_pole_inner_l, bone2Name=Root, swing1LimitDegrees=10, swing2LimitDegrees=10, twistLimitDegrees=10}
 {bone1Name=dyn_flag_inner_l, bone2Name=dyn_pole_inner_l, swing1LimitDegrees=10, swing2LimitDegrees=10, twistLimitDegrees=10}
 {bone1Name=dyn_flag_inner_r, bone2Name=dyn_pole_inner_r, swing1LimitDegrees=10, swing2LimitDegrees=10, twistLimitDegrees=10}
 {bone1Name=dyn_pole_inner_r, bone2Name=Root, swing1LimitDegrees=10, swing2LimitDegrees=10, twistLimitDegrees=10}
 {bone1Name=dyn_flag_outer_r, bone2Name=dyn_pole_outer_r, swing1LimitDegrees=10, swing2LimitDegrees=10, twistLimitDegrees=10}
 {bone1Name=dyn_pole_outer_r, bone2Name=Root, swing1LimitDegrees=10, swing2LimitDegrees=10, twistLimitDegrees=10}
]
